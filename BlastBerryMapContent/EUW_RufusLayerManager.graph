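# Editor utility widget that previews Rufus calendar weeks in the editor by switching data layers.
# On construct or refresh it scans level actors for exactly one BP_Rufus_DataLayerManager, enabling the week controls only if one is found.
# Week buttons append their number to the editable WeekName and match it against event names in the manager's CalendarDataLayerList.
# Matching entries load and show their enabled layers and unload and hide disabled ones via DataLayerEditorSubsystem.
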EditorUtilityWidgetBlueprint /BlastBerryMapContent/EUW_RufusLayerManager
graphs [
 Call Week Swap {nodes [
   n1 = Call Week Swap!() -> (then, Week:String) {is=title K2Node_FunctionEntry}
   n2 = Get LayerManager() -> (LayerManager:BP Rufus Data Layer Manager Object Reference)
   > n3 = Is Valid(InputObject=@n2)
   GetLayerManager(execute=@n3.Is Not Valid)
   n4 = For Each Loop(
    Exec=@n3.Is Valid
    Array:Array of Calendar Data Layer Data Structures=Get CalendarDataLayerList!(self:BP Rufus Data Layer Manager Object Reference=@n2) -> (CalendarDataLayerList:Array of Calendar Data Layer Data Structures)
   ) -> (Array Element:Calendar Data Layer Data Structure)
   n5 = Break Calendar Data Layer Data!(CalendarDataLayerData:"Calendar Data Layer Data Structure (by ref)"=@n4) -> (
    EventName_2_#1:String
    EnabledDataLayers_9_#2:Array of Data Layer Asset Object References
    DisabledDataLayers_10_#3:Array of Data Layer Asset Object References
   ) {is=title K2Node_BreakStruct}
   > Branch(Condition="Equal Exactly (String)"(A=@n1, B=@n5))
   n6 = Get DataLayerEditorSubsystem() -> (ReturnValue:Data Layer Editor Subsystem Object Reference) {is=title K2Node_GetEditorSubsystem}
   n7 = GetDataLayerInstances!(
    self:Data Layer Editor Subsystem Object Reference=@n6
    DataLayerAssets:Array of Data Layer Asset Object References=@n5.EnabledDataLayers_9_#2
   ) -> (ReturnValue:Array of Data Layer Instance Object References)
   > SetDataLayersIsLoadedInEditor!(
    execute
    self:Data Layer Editor Subsystem Object Reference=@n6
    DataLayers:Array of Data Layer Instance Object References=@n7
    bIsLoadedInEditor:bool=true
    bIsFromUserChange:bool=false
   ) -> (then, ReturnValue:bool)
   > SetDataLayersVisibility!(
    execute
    self:Data Layer Editor Subsystem Object Reference=@n6
    DataLayers:Array of Data Layer Instance Object References=@n7
    bIsVisible:bool=true
   ) -> (then)
   n8 = GetDataLayerInstances(self=@n6, DataLayerAssets=@n5.DisabledDataLayers_10_#3)
   > SetDataLayersIsLoadedInEditor(self=@n6, DataLayers=@n8, bIsLoadedInEditor=false, bIsFromUserChange=false)
   > SetDataLayersVisibility(self=@n6, DataLayers=@n8, bIsVisible=false)
  ]}
 EventGraph {nodes [
   "On Clicked (WeekButton_1)"() {is=title K2Node_ComponentBoundEvent}
   n1 = Get WeekName() -> (WeekName:String)
   > Call Week Swap!(execute, self, Week:String=Append(A=@n1, B=1)) -> (then) {is=title K2Node_CallFunction}
   > n2 = Set CalendarWeek(CalendarWeek:int=1) -> (Output_Get:int=0)
   > Set Text(Text="To Text (Integer)"(Value=@n2), self=Get Week() -> (Week:Text Object Reference))
   "On Clicked (WeekButton_2)"() {is=title K2Node_ComponentBoundEvent}
   > Call Week Swap(Week=Append(A=@n1, B=2))
   > n3 = Set CalendarWeek(CalendarWeek=2) -> (Output_Get=0)
   > Set Text(Text="To Text (Integer)"(Value=@n3), self=Get Week())
   "On Clicked (WeekButton_3)"() {is=title K2Node_ComponentBoundEvent}
   > Call Week Swap(Week=Append(A=@n1, B=3))
   > n4 = Set CalendarWeek(CalendarWeek=3) -> (Output_Get=0)
   > Set Text(Text="To Text (Integer)"(Value=@n4), self=Get Week())
   "On Clicked (WeekButton_4)"() {is=title K2Node_ComponentBoundEvent}
   > Call Week Swap(Week=Append(A=@n1, B=4))
   > n5 = Set CalendarWeek(CalendarWeek=4) -> (Output_Get=0)
   > Set Text(Text="To Text (Integer)"(Value=@n5), self=Get Week())
   "On Clicked (RefreshButton)"() {is=title K2Node_ComponentBoundEvent}
   > GetLayerManager()
   n6 = "On Text Changed (EditorUtilityEditableTextBox_147)"!() -> (OutputDelegate, then, Text:"Text (by ref)") {is=title K2Node_ComponentBoundEvent}
   > Set WeekName(WeekName:String="To String (Text)"(InText=@n6.Text)) -> (Output_Get:String)
   GetLayerManager() {is=title K2Node_CustomEvent}
   > n7 = GetAllLevelActors(self=Get EditorActorSubsystem())
   > n8 = For Each Loop(Array:Array of Actor Object References=@n7) -> (Array Element:Actor Object Reference)
   n9 = Get PossibleLayerManagers() -> (PossibleLayerManagers:Array of BP Rufus Data Layer Manager Object References)
   n10 = Switch on Int!(execute:Exec=@n8.Completed, Selection:int=Length(TargetArray:Array of BP Rufus Data Layer Manager Object References=@n9)) -> (Default:Exec, 0:Exec, 1:Exec)
   Set Text(execute=[@n10.0, @n10.Default], Text="LOCTEXT(\"Error\")", self=Get Week())
   > Set ValidLayerManager(ValidLayerManager:bool=false) -> (Output_Get:bool)
   > n11 = PrintString(InString=There are 0 or more than 1 loaded Rufus Data Layer Manager objects in the level., TextColor=(R=1,G=0.9452,B=0.243,A=1), Duration=15)
   ToggleUIElements!(execute=[@n11.then, @n12.then], self, ToggleValue:bool=Get ValidLayerManager() -> (ValidLayerManager:bool)) -> (then)
   > Clear(TargetArray:Array of BP Rufus Data Layer Manager Object References=Get PossibleLayerManagers())
   Set LayerManager(
    execute=@n10.1
    LayerManager:BP Rufus Data Layer Manager Object Reference="Get (a copy)"(Array:Array of BP Rufus Data Layer Manager Object References=@n9) -> (Output:BP Rufus Data Layer Manager Object Reference)
   ) -> (Output_Get:BP Rufus Data Layer Manager Object Reference)
   > Set Text(Text="To Text (Integer)"(Value=Get CalendarWeek() -> (CalendarWeek:int=0)), self=Get Week())
   > n12 = Set ValidLayerManager(ValidLayerManager=true)
   Is Valid(exec=@n8.LoopBody, InputObject=@n8)
   > Branch(Condition="Equal (Class)"(A=Get Class(Object=@n8), B=/BlastBerryMapContent/Game/Blueprints/BP_Rufus_DataLayerManager.BP_Rufus_DataLayerManager_C))
   > Add(
    TargetArray:Array of BP Rufus Data Layer Manager Object References=Get PossibleLayerManagers()
    NewItem:"BP Rufus Data Layer Manager Object Reference (by ref)"=Cast To BP_Rufus_DataLayerManager!(Object=@n8) -> (AsBP Rufus Data Layer Manager:BP Rufus Data Layer Manager Object Reference, bSuccess:bool=~) {is=title K2Node_DynamicCast}
   ) {is=title K2Node_CallArrayFunction}
   Event Construct()
   > Set CalendarWeek(CalendarWeek=0) -> (Output_Get=0)
   > GetLayerManager() {is=title K2Node_CallFunction}
   Event PreConstruct()
   Event Tick()
  ]}
 ToggleUIElements {nodes [
   n1 = ToggleUIElements!() -> (then, ToggleValue:bool=~) {is=title K2Node_FunctionEntry}
   > n2 = Set EUWActors(EUWActors:Array of Vertical Box Object References=Make Array("[0]":Vertical Box Object Reference=Get TopBox() -> (TopBox:Vertical Box Object Reference)) -> (Array:Array of Vertical Box Object References)) -> (Output_Get:Array of Vertical Box Object References)
   > n3 = For Each Loop(Array:Array of Vertical Box Object References=@n2) -> (Array Element:Vertical Box Object Reference)
   > SetIsEnabled(self=@n3, bInIsEnabled=@n1)
  ]}
]
variables [
 CalendarWeek {type=integer}
 EUWActors {type="array<VerticalBox>", default [null]}
 LayerManager {type=BP_Rufus_DataLayerManager_C, default=null}
 PossibleLayerManagers {type="array<BP_Rufus_DataLayerManager_C>", default []}
 ValidLayerManager {type=boolean}
 WeekName {type=string, default=RufusWeek}
]
widgets {
 info {rootWidgetClass=&StackBox, widgetCount=23}
 widgets [
  StackBox_0 {widgetClassPath=&StackBox, widget=StackBox_0, properties {orientation=Orient_Vertical}}
  TextBlock_67 {
   widgetClassPath=&TextBlock
   parent=StackBox_0
   widget=TextBlock_67
   properties {
    justification=Center
    text=Data Layer Manager Loader
    font {
     bForceMonospaced=false
     bMaterialIsStencil=false
     fontMaterial=null
     fontObject=&Roboto
     letterSpacing=0
     monospacedWidth=1
     size=15
     skewAmount=0
     typefaceFontName=Bold
     outlineSettings {
      bApplyOutlineToDropShadows=false
      bMiteredCorners=false
      bSeparateFillAlpha=false
      outlineMaterial=null
      outlineSize=0
      outlineColor=<rgba 0 0 0 1>
     }
    }
   }
   slot {class=StackBoxSlot, properties {horizontalAlignment=HAlign_Center, padding=<ltrb 0 0 0 10>}}
  }
  RefreshButton {
   widgetClassPath=&EditorUtilityButton
   parent=StackBox_0
   widget=RefreshButton
   properties {toolTipText=Try to get a reference to a loaded Data Layer Manager.}
   slot {class=StackBoxSlot, properties {padding=<ltrb 50 0 50 25>}}
  }
  TextBlock_33 {
   widgetClassPath=&TextBlock
   parent=RefreshButton
   widget=TextBlock_33
   properties {
    text=Get Layer Manager
    font {
     bForceMonospaced=false
     bMaterialIsStencil=false
     fontMaterial=null
     fontObject=&Roboto
     letterSpacing=0
     monospacedWidth=1
     size=10.5
     skewAmount=0
     typefaceFontName=Regular
     outlineSettings {
      bApplyOutlineToDropShadows=false
      bMiteredCorners=false
      bSeparateFillAlpha=false
      outlineMaterial=null
      outlineSize=0
      outlineColor=<rgba 0 0 0 1>
     }
    }
   }
   slot {class=ButtonSlot}
  }
  TopBox {widgetClassPath=&VerticalBox, bIsVariable=true, parent=StackBox_0, widget=TopBox, slot {class=StackBoxSlot}}
  HorizontalBox_0 {
   widgetClassPath=&HorizontalBox
   parent=TopBox
   widget=HorizontalBox_0
   properties {toolTipText="The currently loaded Week. If 0, tool hasn't loaded any week."}
   slot {class=VerticalBoxSlot}
  }
  Calendar {
   widgetClassPath=&TextBlock
   parent=HorizontalBox_0
   widget=Calendar
   properties {
    justification=Right
    text="Current Week: "
    font {
     bForceMonospaced=false
     bMaterialIsStencil=false
     fontMaterial=null
     fontObject=&Roboto
     letterSpacing=0
     monospacedWidth=1
     size=10.5
     skewAmount=0
     typefaceFontName=Regular
     outlineSettings {
      bApplyOutlineToDropShadows=false
      bMiteredCorners=false
      bSeparateFillAlpha=false
      outlineMaterial=null
      outlineSize=0
      outlineColor=<rgba 0 0 0 1>
     }
    }
   }
   slot {class=HorizontalBoxSlot, properties {size {sizeRule=Fill, value=1}}}
  }
  Week {
   widgetClassPath=&TextBlock
   bIsVariable=true
   parent=HorizontalBox_0
   widget=Week
   properties {
    text=0
    font {
     bForceMonospaced=false
     bMaterialIsStencil=false
     fontMaterial=null
     fontObject=&Roboto
     letterSpacing=0
     monospacedWidth=1
     size=10.5
     skewAmount=0
     typefaceFontName=Regular
     outlineSettings {
      bApplyOutlineToDropShadows=false
      bMiteredCorners=false
      bSeparateFillAlpha=false
      outlineMaterial=null
      outlineSize=0
      outlineColor=<rgba 0 0 0 1>
     }
    }
   }
   slot {class=HorizontalBoxSlot, properties {size {sizeRule=Fill, value=1}}}
  }
  HorizontalBox_142 {
   widgetClassPath=&HorizontalBox
   parent=TopBox
   widget=HorizontalBox_142
   properties {toolTipText=Determines the name of the Event Names used to load each week.}
   slot {class=VerticalBoxSlot, properties {padding=<ltrb 0 0 0 10>, size {sizeRule=Fill, value=1}}}
  }
  Calendar_1 {
   widgetClassPath=&TextBlock
   parent=HorizontalBox_142
   widget=Calendar_1
   properties {
    justification=Right
    text="Week Name: "
    font {
     bForceMonospaced=false
     bMaterialIsStencil=false
     fontMaterial=null
     fontObject=&Roboto
     letterSpacing=0
     monospacedWidth=1
     size=10.5
     skewAmount=0
     typefaceFontName=Regular
     outlineSettings {
      bApplyOutlineToDropShadows=false
      bMiteredCorners=false
      bSeparateFillAlpha=false
      outlineMaterial=null
      outlineSize=0
      outlineColor=<rgba 0 0 0 1>
     }
    }
   }
   slot {class=HorizontalBoxSlot, properties {verticalAlignment=VAlign_Center, size {sizeRule=Fill, value=1}}}
  }
  EditorUtilityEditableTextBox_147 {
   widgetClassPath=&EditorUtilityEditableTextBox
   bIsVariable=true
   parent=HorizontalBox_142
   widget=EditorUtilityEditableTextBox_147
   properties {hintText=I.e. RufusWeek}
   slot {class=HorizontalBoxSlot, properties {horizontalAlignment=HAlign_Left, size {sizeRule=Fill, value=1}}}
  }
  TextBlock_190 {
   widgetClassPath=&TextBlock
   parent=TopBox
   widget=TextBlock_190
   properties {
    justification=Center
    text=Week 1
    font {
     bForceMonospaced=false
     bMaterialIsStencil=false
     fontMaterial=null
     fontObject=&Roboto
     letterSpacing=0
     monospacedWidth=1
     size=10.5
     skewAmount=0
     typefaceFontName=Regular
     outlineSettings {
      bApplyOutlineToDropShadows=false
      bMiteredCorners=false
      bSeparateFillAlpha=false
      outlineMaterial=null
      outlineSize=0
      outlineColor=<rgba 0 0 0 1>
     }
    }
   }
   slot {class=VerticalBoxSlot, properties {horizontalAlignment=HAlign_Center, verticalAlignment=VAlign_Center}}
  }
  WeekButton_1 {
   widgetClassPath=&EditorUtilityButton
   parent=TopBox
   widget=WeekButton_1
   properties {toolTipText="Load week 1 of the current \"Week Name\". This will load and unload the proper layers based on the assigned row."}
   slot {class=VerticalBoxSlot, properties {padding=<ltrb 50 5 50 10>}}
  }
  TextBlock_302 {
   widgetClassPath=&TextBlock
   parent=WeekButton_1
   widget=TextBlock_302
   properties {
    text=Load
    font {
     bForceMonospaced=false
     bMaterialIsStencil=false
     fontMaterial=null
     fontObject=&Roboto
     letterSpacing=0
     monospacedWidth=1
     size=10.5
     skewAmount=0
     typefaceFontName=Regular
     outlineSettings {
      bApplyOutlineToDropShadows=false
      bMiteredCorners=false
      bSeparateFillAlpha=false
      outlineMaterial=null
      outlineSize=0
      outlineColor=<rgba 0 0 0 1>
     }
    }
   }
   slot {class=ButtonSlot}
  }
  TextBlock {
   widgetClassPath=&TextBlock
   parent=TopBox
   widget=TextBlock
   properties {
    justification=Center
    text=Week 2
    font {
     bForceMonospaced=false
     bMaterialIsStencil=false
     fontMaterial=null
     fontObject=&Roboto
     letterSpacing=0
     monospacedWidth=1
     size=10.5
     skewAmount=0
     typefaceFontName=Regular
     outlineSettings {
      bApplyOutlineToDropShadows=false
      bMiteredCorners=false
      bSeparateFillAlpha=false
      outlineMaterial=null
      outlineSize=0
      outlineColor=<rgba 0 0 0 1>
     }
    }
   }
   slot {class=VerticalBoxSlot, properties {horizontalAlignment=HAlign_Center, verticalAlignment=VAlign_Center}}
  }
  WeekButton_2 {
   widgetClassPath=&EditorUtilityButton
   parent=TopBox
   widget=WeekButton_2
   properties {toolTipText="Load week 2 of the current \"Week Name\". This will load and unload the proper layers based on the assigned row."}
   slot {class=VerticalBoxSlot, properties {padding=<ltrb 50 5 50 10>}}
  }
  TextBlock_1 {
   widgetClassPath=&TextBlock
   parent=WeekButton_2
   widget=TextBlock_1
   properties {
    text=Load
    font {
     bForceMonospaced=false
     bMaterialIsStencil=false
     fontMaterial=null
     fontObject=&Roboto
     letterSpacing=0
     monospacedWidth=1
     size=10.5
     skewAmount=0
     typefaceFontName=Regular
     outlineSettings {
      bApplyOutlineToDropShadows=false
      bMiteredCorners=false
      bSeparateFillAlpha=false
      outlineMaterial=null
      outlineSize=0
      outlineColor=<rgba 0 0 0 1>
     }
    }
   }
   slot {class=ButtonSlot}
  }
  TextBlock_2 {
   widgetClassPath=&TextBlock
   parent=TopBox
   widget=TextBlock_2
   properties {
    justification=Center
    text=Week 3
    font {
     bForceMonospaced=false
     bMaterialIsStencil=false
     fontMaterial=null
     fontObject=&Roboto
     letterSpacing=0
     monospacedWidth=1
     size=10.5
     skewAmount=0
     typefaceFontName=Regular
     outlineSettings {
      bApplyOutlineToDropShadows=false
      bMiteredCorners=false
      bSeparateFillAlpha=false
      outlineMaterial=null
      outlineSize=0
      outlineColor=<rgba 0 0 0 1>
     }
    }
   }
   slot {class=VerticalBoxSlot, properties {horizontalAlignment=HAlign_Center, verticalAlignment=VAlign_Center}}
  }
  WeekButton_3 {
   widgetClassPath=&EditorUtilityButton
   parent=TopBox
   widget=WeekButton_3
   properties {toolTipText="Load week 3 of the current \"Week Name\". This will load and unload the proper layers based on the assigned row."}
   slot {class=VerticalBoxSlot, properties {padding=<ltrb 50 5 50 10>}}
  }
  TextBlock_4 {
   widgetClassPath=&TextBlock
   parent=WeekButton_3
   widget=TextBlock_4
   properties {
    text=Load
    font {
     bForceMonospaced=false
     bMaterialIsStencil=false
     fontMaterial=null
     fontObject=&Roboto
     letterSpacing=0
     monospacedWidth=1
     size=10.5
     skewAmount=0
     typefaceFontName=Regular
     outlineSettings {
      bApplyOutlineToDropShadows=false
      bMiteredCorners=false
      bSeparateFillAlpha=false
      outlineMaterial=null
      outlineSize=0
      outlineColor=<rgba 0 0 0 1>
     }
    }
   }
   slot {class=ButtonSlot}
  }
  TextBlock_3 {
   widgetClassPath=&TextBlock
   parent=TopBox
   widget=TextBlock_3
   properties {
    justification=Center
    text=Week 4
    font {
     bForceMonospaced=false
     bMaterialIsStencil=false
     fontMaterial=null
     fontObject=&Roboto
     letterSpacing=0
     monospacedWidth=1
     size=10.5
     skewAmount=0
     typefaceFontName=Regular
     outlineSettings {
      bApplyOutlineToDropShadows=false
      bMiteredCorners=false
      bSeparateFillAlpha=false
      outlineMaterial=null
      outlineSize=0
      outlineColor=<rgba 0 0 0 1>
     }
    }
   }
   slot {class=VerticalBoxSlot, properties {horizontalAlignment=HAlign_Center, verticalAlignment=VAlign_Center}}
  }
  WeekButton_4 {
   widgetClassPath=&EditorUtilityButton
   parent=TopBox
   widget=WeekButton_4
   properties {toolTipText="Load week 4 of the current \"Week Name\". This will load and unload the proper layers based on the assigned row."}
   slot {class=VerticalBoxSlot, properties {padding=<ltrb 50 5 50 10>}}
  }
  TextBlock_5 {
   widgetClassPath=&TextBlock
   parent=WeekButton_4
   widget=TextBlock_5
   properties {
    text=Load
    font {
     bForceMonospaced=false
     bMaterialIsStencil=false
     fontMaterial=null
     fontObject=&Roboto
     letterSpacing=0
     monospacedWidth=1
     size=10.5
     skewAmount=0
     typefaceFontName=Regular
     outlineSettings {
      bApplyOutlineToDropShadows=false
      bMiteredCorners=false
      bSeparateFillAlpha=false
      outlineMaterial=null
      outlineSize=0
      outlineColor=<rgba 0 0 0 1>
     }
    }
   }
   slot {class=ButtonSlot}
  }
 ]
}
